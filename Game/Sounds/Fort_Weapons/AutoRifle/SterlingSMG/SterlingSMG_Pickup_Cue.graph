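# Sound cue for picking up the SterlingSMG weapon, classed as loot-collect audio.
# A single wave player feeds the output directly, reusing the WinterTrain assault rifle pickup sample rather than an SMG-specific recording.
# It plays at reduced volume through the shared weapon pickup attenuation, with a concurrency override that allows one instance and blocks new plays while it is active.
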SoundCue /Game/Sounds/Fort_Weapons/AutoRifle/SterlingSMG/SterlingSMG_Pickup_Cue
root=WavePlayer_1
nodes [{id=WavePlayer_1, properties {soundWaveAssetPtr=/Game/Sounds/Fort_Weapons/AutoRifle/WinterTrain/Train_Assault_Pickup_01}}]
properties {
 attenuationSettings=&Fort_Weapon_Pickup_Att
 bOverrideConcurrency=true
 maxDistance=3500
 soundClassObject=&Fort_Loot_Collect
 volumeMultiplier=0.4154
 concurrencyOverrides {maxCount=1, resolutionRule=PreventNew}
}
